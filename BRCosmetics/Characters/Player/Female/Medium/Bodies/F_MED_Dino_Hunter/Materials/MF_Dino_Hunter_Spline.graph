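# Material function for the Dino Hunter character that renders spline ribbons as glowing, color-banded strips over incoming material attributes.
# It reuses Neon's MF_splineThickenM for ribbon width, flatten and normals, adding a camera offset to its world position offset.
# Chained If and Lerp steps threshold vertex color green into hard bands of five tunable colors, black below the first.
# The result is multiplied into emissive, with base color set to black.
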MaterialFunction /BRCosmetics/Characters/Player/Female/Medium/Bodies/F_MED_Dino_Hunter/Materials/MF_Dino_Hunter_Spline
graph {nodes [
  n1 = ScalarParameter() {defaultValue=1, group=DinoHunter, parameterName=DH_CameraOffsetAmount}
  n2 = MaterialFunctionCall!(
   widthTip=ScalarParameter() {defaultValue=1, group=DinoHunter, parameterName=DN_RibbonWidthTip}
   Flatten=ScalarParameter() {group=DinoHunter, parameterName=DN_RibbonFlatten}
   widthBase=ScalarParameter() {defaultValue=1, group=DinoHunter, parameterName=DN_RibbonWidthBase}
  ) {materialFunction=/Game/Characters/Player/Male/Medium/Bodies/M_MED_Neon/Materials/MF_splineThickenM}
  n3 = VertexColor()
  n4 = Constant() {r=1}
  n5 = LinearInterpolate(
   B=VectorParameter() {group=DinoHunter, parameterName=DH_SplineColor2, defaultValue=<rgba 1 0 0.1564 1>}.RGB
   Alpha=If(A=@n3.G, B=Constant() {r=0.2}, "A > B"=@n4, "A == B"=@n4, "A < B"=Constant())
  )
  n6 = VectorParameter() {group=DinoHunter, parameterName=DH_SplineColor3, defaultValue=<rgba 1 0.1922 0 1>}
  n7 = Constant() {r=1}
  n8 = If(A=@n3.G, B=Constant() {r=0.4}, "A > B"=@n7, "A == B"=@n7, "A < B"=Constant())
  n9 = Constant() {r=0.6}
  n10 = Constant() {r=1}
  n11 = Constant()
  n12 = Constant() {r=1}
  n13 = LinearInterpolate(
   A=LinearInterpolate(
    A=LinearInterpolate(A=@n5, B=@n6.RGB, Alpha=@n8)
    B=VectorParameter() {group=DinoHunter, parameterName=DH_SplineColor4, defaultValue=<rgba 1 0.7352 0 1>}.RGB
    Alpha=If(A=@n3.G, B=@n9, "A > B"=@n10, "A == B"=@n10, "A < B"=@n11)
   )
   B=VectorParameter() {group=DinoHunter, parameterName=DH_SplineColor5, defaultValue=<rgba 0.2334 1 0 1>}.RGB
   Alpha=If(A=@n3.G, B=Constant() {r=0.8}, "A > B"=@n12, "A == B"=@n12, "A < B"=Constant())
  )
  n14 = VectorParameter() {group=DinoHunter, parameterName=DH_SplineColor, defaultValue=<rgba 0.4062 0 0.3859 1>}
  n15 = Constant() {r=1}
  n16 = If(A=@n3.G, B=Constant() {r=1}, "A > B"=@n15, "A == B"=@n15, "A < B"=Constant())
  n17 = SetMaterialAttributes!(
   MaterialAttributes=FunctionInput() {inputType=FunctionInput_MaterialAttributes}
   WorldPositionOffset=Add(
    A=MaterialFunctionCall(Offset Amount=@n1) {materialFunction=/Engine/Functions/Engine_MaterialFunctions02/WorldPositionOffset/CameraOffset}.Camera Offset
    B=@n2.worldPositionOffset
   )
   Normal=@n2.normal
   EmissiveColor=Multiply(A=LinearInterpolate(A=@n13, B=@n14.RGB, Alpha=@n16), B=ScalarParameter() {defaultValue=1.5, group=DinoHunter, parameterName=DH_EmissiveMult})
   BaseColor=Constant()
  ) {attributeSetTypes [
    #1
    #2
    #3
    #4
   ]}
  FunctionOutput(None=@n17)
 ]}
