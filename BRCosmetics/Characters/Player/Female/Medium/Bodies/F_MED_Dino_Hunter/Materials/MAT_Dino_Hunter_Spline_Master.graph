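# Master material for the Dino Hunter female medium player body cosmetic, built for skeletal meshes with morph targets and mesh deformers.
# The shared character PreFX function supplies base material attributes, which pass through the MF_Dino_Hunter_Spline function for the character-specific effect.
# The shared PostFX function finishes the result, receiving PreFX outputs such as AO alongside it.
# It outputs through material attributes with Default Lit shading.
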Material /BRCosmetics/Characters/Player/Female/Medium/Bodies/F_MED_Dino_Hunter/Materials/MAT_Dino_Hunter_Spline_Master
graph {
 properties {
  bUseMaterialAttributes=true
  bUsedWithMeshDeformer=true
  bUsedWithMorphTargets=true
  bUsedWithSkeletalMesh=true
  pixelDepthOffsetMode=PDOM_Legacy
  usedShadingModels=Default Lit
 }
 nodes [
  n1 = MaterialFunctionCall() {materialFunction=&BaseFN_CharacterFunction_PreFX}
  MaterialFunctionCall(
   AO=@n1.AO
   Distance Blend=@n1.AO
   Fresnel=@n1.AO
   Gmap=@n1.AO
   Pre Skinned Local Position=@n1.AO
   Material=MaterialFunctionCall!(In=@n1.Result) {materialFunction=&/MF_Dino_Hunter_Spline}.Result
  ) {materialFunction=&BaseFN_CharacterFunction_PostFX}
 ]
}
