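# Physics asset for the Dino Hunter female medium character's belt, setting up physical bodies for a canteen, two pouches and a rope.
# Canteen and pouches are capsules, each constrained to a tiny base capsule, while the rope is two boxes chained from its own base.
# Angular limits keep the pouches nearly rigid, let the canteen swing moderately and give the rope the most play.
# A right-thigh capsule has no constraints, apparently serving as a leg collider.
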PhysicsAsset /BRCosmetics/Characters/Player/Female/Medium/Bodies/F_MED_Dino_Hunter/Meshes/F_MED_Dino_Hunter_Belt_Physics
bodies {
 dyn_canteen_base_r [{shapeType=Capsule, length=0.505, radius=0.505, rotation=<rot 0 -90 0>}]
 dyn_canteen_r [{shapeType=Capsule, length=10.25, radius=5, center=<5.8435 2.5 0>, rotation=<rot 90 0 0>}]
 dyn_pouch_bk_base_r [{shapeType=Capsule, length=0.505, radius=0.505, rotation=<rot 0 -90 0>}]
 dyn_pouch_bk_r [{shapeType=Capsule, length=3.5, radius=3.5, center=<4 0 0>, rotation=<rot 90 0 0>}]
 dyn_pouch_fr_base_r [{shapeType=Capsule, length=0.505, radius=0.505, rotation=<rot 0 -90 0>}]
 dyn_pouch_fr_r [{shapeType=Capsule, length=3.5, radius=3.5, center=<4 0 0>, rotation=<rot 90 0 0>}]
 dyn_rope_a [{shapeType=Box, extentX=11, extentY=4, extentZ=15.5, center=<5.5 0 0>}]
 dyn_rope_a_base [{shapeType=Capsule, length=0.505, radius=0.505, rotation=<rot 0 -90 0>}]
 dyn_rope_b [{shapeType=Box, extentX=15, extentY=4, extentZ=15.5, center=<7.3694 -0.0001 0>}]
 thigh_r [{shapeType=Capsule, length=37.75, radius=7.25, center=<25.7844 0.0003 -0.9156>, rotation=<rot 88.4313 -0.0001 0>}]
}
constraints [
 {bone1Name=dyn_rope_a, bone2Name=dyn_rope_a_base, swing1LimitDegrees=25, swing2LimitDegrees=20, twistLimitDegrees=20}
 {bone1Name=dyn_rope_b, bone2Name=dyn_rope_a, swing1LimitDegrees=25, swing2LimitDegrees=20, twistLimitDegrees=20}
 {bone1Name=dyn_canteen_r, bone2Name=dyn_canteen_base_r, swing1LimitDegrees=15, swing2LimitDegrees=5, twistLimitDegrees=10}
 {bone1Name=dyn_pouch_fr_r, bone2Name=dyn_pouch_fr_base_r, swing1LimitDegrees=2, swing2LimitDegrees=5, twistLimitDegrees=2}
 {bone1Name=dyn_pouch_bk_r, bone2Name=dyn_pouch_bk_base_r, swing1LimitDegrees=2, swing2LimitDegrees=5, twistLimitDegrees=2}
]
