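# Physics asset for the necklace on the female medium Dino Hunter body, giving its bones capsule bodies for physics-driven motion.
# Two dyn_necklace capsules form a chain from spine_05, with tight angular limits and twist locked at the first link.
# A large spine_05 capsule and a tiny neck_01 capsule represent the upper body, joined by a constraint with free swing and twist.
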PhysicsAsset /BRCosmetics/Characters/Player/Female/Medium/Bodies/F_MED_Dino_Hunter/Meshes/F_MED_Dino_Hunter_Necklace_Physics
bodies {
 dyn_necklace_a [{shapeType=Capsule, length=3.5, radius=1.25, center=<3 0 0>, rotation=<rot 90 0 0>}]
 dyn_necklace_b [{shapeType=Capsule, length=5.5, radius=1.25, center=<4 0 0>, rotation=<rot 90 0 0>}]
 neck_01 [{shapeType=Capsule, length=0.505, radius=0.505, rotation=<rot 0 -90 0>}]
 spine_05 [{shapeType=Capsule, length=6.5, radius=9.75, center=<0.0534 -1.2035 -0.0002>}]
}
constraints [
 {
  bone1Name=neck_01
  bone2Name=spine_05
  swing1LimitDegrees=45
  swing1Motion=Free
  swing2LimitDegrees=45
  swing2Motion=Free
  twistLimitDegrees=45
  twistMotion=Free
 }
 {bone1Name=dyn_necklace_a, bone2Name=spine_05, swing1LimitDegrees=15, swing2LimitDegrees=1, twistLimitDegrees=45, twistMotion=Locked}
 {bone1Name=dyn_necklace_b, bone2Name=dyn_necklace_a, swing1LimitDegrees=15, swing2LimitDegrees=5, twistLimitDegrees=1}
]
