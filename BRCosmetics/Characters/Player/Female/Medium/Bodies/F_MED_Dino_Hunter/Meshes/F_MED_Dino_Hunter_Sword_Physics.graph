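# Physics asset for the sword on the F_MED_Dino_Hunter female player cosmetic body, giving it physics bodies on dynamic bones.
# A capsule body on dyn_sword covers the blade, and a tiny capsule on dyn_sword_base serves as the attachment point.
# A single constraint joins the two bones with small swing and twist limits, so the sword can move only slightly relative to its base.
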PhysicsAsset /BRCosmetics/Characters/Player/Female/Medium/Bodies/F_MED_Dino_Hunter/Meshes/F_MED_Dino_Hunter_Sword_Physics
bodies {
 dyn_sword [{shapeType=Capsule, length=20.8263, radius=5.5621, center=<2 0 0>, rotation=<rot 90 0 0>}]
 dyn_sword_base [{shapeType=Capsule, length=0.505, radius=0.505, rotation=<rot 0 -90 0>}]
}
constraints [{bone1Name=dyn_sword, bone2Name=dyn_sword_base, swing1LimitDegrees=10, swing2LimitDegrees=5, twistLimitDegrees=2}]
