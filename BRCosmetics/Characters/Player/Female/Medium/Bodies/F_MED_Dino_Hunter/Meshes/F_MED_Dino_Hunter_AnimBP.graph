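# Secondary-motion Animation Blueprint for the female Dino Hunter body, copying its attached parent's pose and animating costume and gear bones.
# Weighted constraints place volume, shoulder-horn, shoestring, belt and shirt bones.
# Per-update joint yaw deltas from reference pose drive additive elbow, thigh, belt and sword correctives.
# RigidBody and AnimDynamics nodes finish the simulation; calf and upper-arm-driven horn correctives are disconnected from the output.
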AnimBlueprint /BRCosmetics/Characters/Player/Female/Medium/Bodies/F_MED_Dino_Hunter/Meshes/F_MED_Dino_Hunter_AnimBP
target_skeleton=&/F_MED_Dino_Hunter_Skeleton
graphs [
 AnimGraph {nodes [
   n1 = "Constraint - dyn_volume_base_calf_l (PP)"!(ComponentPose, Alpha=1, ConstraintWeights_0=0.5, ConstraintWeights_1:float=0.5) -> (Pose) {
    alphaBoolBlend {blendOption=Linear}
    boneToModify {boneName=dyn_volume_base_calf_l}
    constraintSetup [{targetBone {boneName=thigh_l}}, {targetBone {boneName=calf_l}}]
    constraintWeights [1,1]
   }
   n2 = "Constraint - dyn_volume_base_calf_r (PP)"!(ComponentPose=@n1, Alpha=1, ConstraintWeights_0=0.5, ConstraintWeights_1:float=0.5) -> (Pose) {
    alphaBoolBlend {blendOption=Linear}
    boneToModify {boneName=dyn_volume_base_calf_r}
    constraintSetup [{targetBone {boneName=thigh_r}}, {targetBone {boneName=calf_r}}]
    constraintWeights [1,1]
   }
   n3 = "Transform (Modify) Bone - Bone: dyn_volume_calf_l"(ComponentPose=@n2, Alpha=Get knee_rot_l() -> (knee_rot_l:double), Translation="1,3,0", Rotation="0,0,0", Scale="1.5,1,1.5") {
    rotationMode=BMM_Additive
    rotationSpace=BCS_BoneSpace
    scaleMode=BMM_Additive
    scaleSpace=BCS_BoneSpace
    translationMode=BMM_Additive
    translationSpace=BCS_BoneSpace
    alphaBoolBlend {blendOption=Linear}
    alphaScaleBiasClamp {bMapRange=true, inRange {max=80}}
    boneToModify {boneName=dyn_volume_calf_l}
    scale {x=1.5, z=1.5}
    translation=<1 3>
   }
   "Transform (Modify) Bone - Bone: dyn_volume_calf_r"(ComponentPose=@n3, Alpha=Get knee_rot_r() -> (knee_rot_r:double), Translation="-1,-3,0", Rotation="0,0,0", Scale="1.5,1,1.5") {
    rotationMode=BMM_Additive
    rotationSpace=BCS_BoneSpace
    scaleMode=BMM_Additive
    scaleSpace=BCS_BoneSpace
    translationMode=BMM_Additive
    translationSpace=BCS_BoneSpace
    alphaBoolBlend {blendOption=Linear}
    alphaScaleBiasClamp {bMapRange=true, inRange {max=80}}
    boneToModify {boneName=dyn_volume_calf_r}
    scale {x=1.5, z=1.5}
    translation=<-1 -3>
   }
   "Transform (Modify) Bone - Bone: dyn_shldrCapHorn_g_r"(Alpha=Get uparm_rot_r() -> (uparm_rot_r:double), Translation="1,0,-1", Rotation="0,-40,0", Scale="1,1,1") {
    rotationMode=BMM_Additive
    rotationSpace=BCS_BoneSpace
    scaleMode=BMM_Additive
    scaleSpace=BCS_BoneSpace
    translationMode=BMM_Additive
    translationSpace=BCS_BoneSpace
    alphaBoolBlend {blendOption=Linear}
    alphaScaleBiasClamp {bMapRange=true, interpSpeedDecreasing=25, interpSpeedIncreasing=25, inRange {max=85}}
    boneToModify {boneName=dyn_shldrCapHorn_g_r}
    rotation {yaw=-40}
    translation {x=1, z=-1}
   }
   n4 = "Constraint - dyn_volume_base_lowerarm_l (PP)"!(
    ComponentPose=Local To Component(LocalPose=Copy Pose From Mesh() {bUseAttachedParent=true})
    Alpha=1
    ConstraintWeights_0=0.65
    ConstraintWeights_1:float=0.35
   ) -> (Pose) {
    alphaBoolBlend {blendOption=Linear}
    boneToModify {boneName=dyn_volume_base_lowerarm_l}
    constraintSetup [{targetBone {boneName=upperarm_l}}, {targetBone {boneName=lowerarm_l}}]
    constraintWeights [1,1]
   }
   n5 = "Constraint - dyn_volume_base_lowerarm_r (PP)"!(ComponentPose=@n4, Alpha=1, ConstraintWeights_0=0.65, ConstraintWeights_1:float=0.35) -> (Pose) {
    alphaBoolBlend {blendOption=Linear}
    boneToModify {boneName=dyn_volume_base_lowerarm_r}
    constraintSetup [{targetBone {boneName=upperarm_r}}, {targetBone {boneName=lowerarm_r}}]
    constraintWeights [1,1]
   }
   n6 = "Transform (Modify) Bone - Bone: dyn_volume_lowerarm_l"(ComponentPose=@n5, Alpha=Get elbow_rot_l() -> (elbow_rot_l:double), Translation="0,-2,0", Rotation="0,0,0", Scale="0.3,1,1") {
    rotationMode=BMM_Additive
    rotationSpace=BCS_BoneSpace
    scaleMode=BMM_Additive
    scaleSpace=BCS_BoneSpace
    translationMode=BMM_Additive
    translationSpace=BCS_BoneSpace
    alphaBoolBlend {blendOption=Linear}
    alphaScaleBiasClamp {bMapRange=true, inRange {max=80}}
    boneToModify {boneName=dyn_volume_lowerarm_l}
    scale {x=0.3}
    translation {y=-2}
   }
   n7 = "Transform (Modify) Bone - Bone: dyn_volume_lowerarm_r"(ComponentPose=@n6, Alpha=Get elbow_rot_r() -> (elbow_rot_r:double), Translation="0,2,0", Rotation="0,0,0", Scale="0.3,1,1") {
    rotationMode=BMM_Additive
    rotationSpace=BCS_BoneSpace
    scaleMode=BMM_Additive
    scaleSpace=BCS_BoneSpace
    translationMode=BMM_Additive
    translationSpace=BCS_BoneSpace
    alphaBoolBlend {blendOption=Linear}
    alphaScaleBiasClamp {bMapRange=true, inRange {max=80}}
    boneToModify {boneName=dyn_volume_lowerarm_r}
    scale {x=0.3}
    translation {y=2}
   }
   n8 = "Constraint - dyn_volume_base_thigh_l (PP)"!(ComponentPose=@n7, Alpha=1, ConstraintWeights_0=0.6, ConstraintWeights_1:float=0.4) -> (Pose) {
    alphaBoolBlend {blendOption=Linear}
    boneToModify {boneName=dyn_volume_base_thigh_l}
    constraintSetup [{targetBone {boneName=pelvis}}, {targetBone {boneName=thigh_l}}]
    constraintWeights [1,1]
   }
   n9 = "Constraint - dyn_volume_base_thigh_r (PP)"!(ComponentPose=@n8, Alpha=1, ConstraintWeights_0=0.6, ConstraintWeights_1:float=0.4) -> (Pose) {
    alphaBoolBlend {blendOption=Linear}
    boneToModify {boneName=dyn_volume_base_thigh_r}
    constraintSetup [{targetBone {boneName=pelvis}}, {targetBone {boneName=thigh_r}}]
    constraintWeights [1,1]
   }
   n10 = Get thigh_rot_l()
   n11 = "Transform (Modify) Bone - Bone: dyn_volume_thigh_l"(ComponentPose=@n9, Alpha=@n10, Translation="-4,0,0", Rotation="0,0,0", Scale="1,1,1") {
    rotationMode=BMM_Additive
    rotationSpace=BCS_BoneSpace
    scaleMode=BMM_Additive
    scaleSpace=BCS_BoneSpace
    translationMode=BMM_Additive
    translationSpace=BCS_BoneSpace
    alphaBoolBlend {blendOption=Linear}
    alphaScaleBiasClamp {bMapRange=true, inRange {max=0, min=-100}, outRange {max=0, min=1}}
    boneToModify {boneName=dyn_volume_thigh_l}
    translation {x=-4}
   }
   n12 = "Transform (Modify) Bone - Bone: dyn_volume_thigh_l"(ComponentPose=@n11, Alpha=@n10, Translation="0,-4,0", Rotation="0,0,0", Scale="1,1,1") {
    rotationMode=BMM_Additive
    rotationSpace=BCS_BoneSpace
    scaleMode=BMM_Additive
    scaleSpace=BCS_BoneSpace
    translationMode=BMM_Additive
    translationSpace=BCS_BoneSpace
    alphaBoolBlend {blendOption=Linear}
    alphaScaleBiasClamp {bMapRange=true, inRange {max=50}}
    boneToModify {boneName=dyn_volume_thigh_l}
    translation {y=-4}
   }
   n13 = Get thigh_rot_r()
   n14 = "Transform (Modify) Bone - Bone: dyn_volume_thigh_r"(ComponentPose=@n12, Alpha=@n13, Translation="4,0,0", Rotation="0,0,0", Scale="1,1,1") {
    rotationMode=BMM_Additive
    rotationSpace=BCS_BoneSpace
    scaleMode=BMM_Additive
    scaleSpace=BCS_BoneSpace
    translationMode=BMM_Additive
    translationSpace=BCS_BoneSpace
    alphaBoolBlend {blendOption=Linear}
    alphaScaleBiasClamp {bMapRange=true, inRange {max=0, min=-100}, outRange {max=0, min=1}}
    boneToModify {boneName=dyn_volume_thigh_r}
    translation {x=4}
   }
   n15 = "Transform (Modify) Bone - Bone: dyn_volume_thigh_r"(ComponentPose=@n14, Alpha=@n13, Translation="0,4,0", Rotation="0,0,0", Scale="1,1,1") {
    rotationMode=BMM_Additive
    rotationSpace=BCS_BoneSpace
    scaleMode=BMM_Additive
    scaleSpace=BCS_BoneSpace
    translationMode=BMM_Additive
    translationSpace=BCS_BoneSpace
    alphaBoolBlend {blendOption=Linear}
    alphaScaleBiasClamp {bMapRange=true, inRange {max=50}}
    boneToModify {boneName=dyn_volume_thigh_r}
    translation {y=4}
   }
   n16 = "Constraint - dyn_shldrCapHorn_a_r (PPPP)"!(
    ComponentPose=@n15
    Alpha=1
    ConstraintWeights_0=0.3527
    ConstraintWeights_1:float=0.6182
    ConstraintWeights_2:float=0
    ConstraintWeights_3:float=0.0291
   ) -> (Pose) {
    alphaBoolBlend {blendOption=Linear}
    boneToModify {boneName=dyn_shldrCapHorn_a_r}
    constraintSetup [
     {targetBone {boneName=clavicle_r}}
     {targetBone {boneName=upperarm_twist_01_r}}
     {targetBone {boneName=upperarm_r}}
     {targetBone {boneName=spine_05}}
    ]
    constraintWeights [1,1,1,1]
   }
   n17 = "Constraint - dyn_shldrCapHorn_b_r (PPPP)"!(
    ComponentPose=@n16
    Alpha=1
    ConstraintWeights_0=0.4759
    ConstraintWeights_1:float=0.5062
    ConstraintWeights_2:float=0
    ConstraintWeights_3:float=0.0179
   ) -> (Pose) {
    alphaBoolBlend {blendOption=Linear}
    boneToModify {boneName=dyn_shldrCapHorn_b_r}
    constraintSetup [
     {targetBone {boneName=clavicle_r}}
     {targetBone {boneName=upperarm_twist_01_r}}
     {targetBone {boneName=upperarm_r}}
     {targetBone {boneName=spine_05}}
    ]
    constraintWeights [1,1,1,1]
   }
   n18 = "Constraint - dyn_shldrCapHorn_c_r (PPPP)"!(
    ComponentPose=@n17
    Alpha=1
    ConstraintWeights_0=0.6679
    ConstraintWeights_1:float=0.1987
    ConstraintWeights_2:float=0
    ConstraintWeights_3:float=0.1335
   ) -> (Pose) {
    alphaBoolBlend {blendOption=Linear}
    boneToModify {boneName=dyn_shldrCapHorn_c_r}
    constraintSetup [
     {targetBone {boneName=clavicle_r}}
     {targetBone {boneName=upperarm_twist_01_r}}
     {targetBone {boneName=upperarm_r}}
     {targetBone {boneName=spine_05}}
    ]
    constraintWeights [1,1,1,1]
   }
   n19 = "Constraint - dyn_shldrCapHorn_d_r (PPPP)"!(
    ComponentPose=@n18
    Alpha=1
    ConstraintWeights_0=0.6327
    ConstraintWeights_1:float=0.209
    ConstraintWeights_2:float=0
    ConstraintWeights_3:float=0.1583
   ) -> (Pose) {
    alphaBoolBlend {blendOption=Linear}
    boneToModify {boneName=dyn_shldrCapHorn_d_r}
    constraintSetup [
     {targetBone {boneName=clavicle_r}}
     {targetBone {boneName=upperarm_twist_01_r}}
     {targetBone {boneName=upperarm_r}}
     {targetBone {boneName=spine_05}}
    ]
    constraintWeights [1,1,1,1]
   }
   n20 = "Constraint - dyn_shldrCapHorn_e_r (PPPP)"!(
    ComponentPose=@n19
    Alpha=1
    ConstraintWeights_0=0.3559
    ConstraintWeights_1:float=0.6319
    ConstraintWeights_2:float=0
    ConstraintWeights_3:float=0.0122
   ) -> (Pose) {
    alphaBoolBlend {blendOption=Linear}
    boneToModify {boneName=dyn_shldrCapHorn_e_r}
    constraintSetup [
     {targetBone {boneName=clavicle_r}}
     {targetBone {boneName=upperarm_twist_01_r}}
     {targetBone {boneName=upperarm_r}}
     {targetBone {boneName=spine_05}}
    ]
    constraintWeights [1,1,1,1]
   }
   n21 = "Constraint - dyn_shldrCapHorn_f_r (PPPP)"!(
    ComponentPose=@n20
    Alpha=1
    ConstraintWeights_0=0.6995
    ConstraintWeights_1:float=0.1331
    ConstraintWeights_2:float=0
    ConstraintWeights_3:float=0.1673
   ) -> (Pose) {
    alphaBoolBlend {blendOption=Linear}
    boneToModify {boneName=dyn_shldrCapHorn_f_r}
    constraintSetup [
     {targetBone {boneName=clavicle_r}}
     {targetBone {boneName=upperarm_twist_01_r}}
     {targetBone {boneName=upperarm_r}}
     {targetBone {boneName=spine_05}}
    ]
    constraintWeights [1,1,1,1]
   }
   n22 = "Constraint - dyn_shldrCapHorn_h_r (PPPP)"!(
    ComponentPose=@n21
    Alpha=1
    ConstraintWeights_0=0.6547
    ConstraintWeights_1:float=0.2865
    ConstraintWeights_2:float=0
    ConstraintWeights_3:float=0.0587
   ) -> (Pose) {
    alphaBoolBlend {blendOption=Linear}
    boneToModify {boneName=dyn_shldrCapHorn_h_r}
    constraintSetup [
     {targetBone {boneName=clavicle_r}}
     {targetBone {boneName=upperarm_twist_01_r}}
     {targetBone {boneName=upperarm_r}}
     {targetBone {boneName=spine_05}}
    ]
    constraintWeights [1,1,1,1]
   }
   n23 = "Constraint - dyn_shldrCapHorn_g_r (PPPP)"!(
    ComponentPose=@n22
    Alpha=1
    ConstraintWeights_0=0.6112
    ConstraintWeights_1:float=0.2378
    ConstraintWeights_2:float=0
    ConstraintWeights_3:float=0.151
   ) -> (Pose) {
    alphaBoolBlend {blendOption=Linear}
    boneToModify {boneName=dyn_shldrCapHorn_g_r}
    constraintSetup [
     {targetBone {boneName=clavicle_r}}
     {targetBone {boneName=upperarm_twist_01_r}}
     {targetBone {boneName=upperarm_r}}
     {targetBone {boneName=spine_05}}
    ]
    constraintWeights [1,1,1,1]
   }
   n24 = "Constraint - dyn_shoeString_l (PPPPP)"!(
    ComponentPose=@n23
    Alpha=1
    ConstraintWeights_0=0
    ConstraintWeights_1:float=0.4
    ConstraintWeights_2:float=0.6
    ConstraintWeights_3:float=0
    ConstraintWeights_4:float=0
   ) -> (Pose) {
    alphaBoolBlend {blendOption=Linear}
    boneToModify {boneName=dyn_shoeString_l}
    constraintSetup [
     {targetBone {boneName=calf_l}}
     {targetBone {boneName=calf_twist_01_l}}
     {targetBone {boneName=foot_l}}
     {targetBone {boneName=ball_l}}
     {targetBone {boneName=calf_twist_02_l}}
    ]
    constraintWeights [1,1,1,1,1]
   }
   n25 = "Constraint - dyn_shoeString_r (PPPPP)"!(
    ComponentPose=@n24
    Alpha=1
    ConstraintWeights_0=0
    ConstraintWeights_1:float=0
    ConstraintWeights_2:float=0.4
    ConstraintWeights_3:float=0.6
    ConstraintWeights_4:float=0
   ) -> (Pose) {
    alphaBoolBlend {blendOption=Linear}
    boneToModify {boneName=dyn_shoeString_r}
    constraintSetup [
     {targetBone {boneName=ball_r}}
     {targetBone {boneName=calf_twist_02_r}}
     {targetBone {boneName=calf_twist_01_r}}
     {targetBone {boneName=foot_r}}
     {targetBone {boneName=calf_r}}
    ]
    constraintWeights [1,1,1,1,1]
   }
   n26 = "Constraint - dyn_belt_fr_base_l (PP)"!(ComponentPose=@n25, Alpha=1, ConstraintWeights_0=0.9, ConstraintWeights_1:float=0.1) -> (Pose) {
    alphaBoolBlend {blendOption=Linear}
    boneToModify {boneName=dyn_belt_fr_base_l}
    constraintSetup [{targetBone {boneName=pelvis}}, {targetBone {boneName=thigh_twist_01_l}}]
    constraintWeights [1,1]
   }
   n27 = "Constraint - dyn_belt_side_base_l (PP)"!(ComponentPose=@n26, Alpha=1, ConstraintWeights_0=0.9, ConstraintWeights_1:float=0.1) -> (Pose) {
    alphaBoolBlend {blendOption=Linear}
    boneToModify {boneName=dyn_belt_side_base_l}
    constraintSetup [{targetBone {boneName=pelvis}}, {targetBone {boneName=thigh_twist_01_l}}]
    constraintWeights [1,1]
   }
   n28 = "Constraint - dyn_belt_bk_base_l (PP)"!(ComponentPose=@n27, Alpha=1, ConstraintWeights_0=0.9, ConstraintWeights_1:float=0.1) -> (Pose) {
    alphaBoolBlend {blendOption=Linear}
    boneToModify {boneName=dyn_belt_bk_base_l}
    constraintSetup [{targetBone {boneName=pelvis}}, {targetBone {boneName=thigh_twist_01_l}}]
    constraintWeights [1,1]
   }
   n29 = "Constraint - dyn_belt_fr_base_r (PP)"!(ComponentPose=@n28, Alpha=1, ConstraintWeights_0=0.5, ConstraintWeights_1:float=0.5) -> (Pose) {
    alphaBoolBlend {blendOption=Linear}
    boneToModify {boneName=dyn_belt_fr_base_r}
    constraintSetup [{targetBone {boneName=pelvis}}, {targetBone {boneName=thigh_twist_01_r}}]
    constraintWeights [1,1]
   }
   n30 = "Constraint - dyn_belt_side_base_r (PP)"!(ComponentPose=@n29, Alpha=1, ConstraintWeights_0=0.1, ConstraintWeights_1:float=0.9) -> (Pose) {
    alphaBoolBlend {blendOption=Linear}
    boneToModify {boneName=dyn_belt_side_base_r}
    constraintSetup [{targetBone {boneName=pelvis}}, {targetBone {boneName=thigh_twist_01_r}}]
    constraintWeights [1,1]
   }
   n31 = "Constraint - dyn_belt_bk_base_r (PP)"!(ComponentPose=@n30, Alpha=1, ConstraintWeights_0=0.9, ConstraintWeights_1:float=0.1) -> (Pose) {
    alphaBoolBlend {blendOption=Linear}
    boneToModify {boneName=dyn_belt_bk_base_r}
    constraintSetup [{targetBone {boneName=pelvis}}, {targetBone {boneName=thigh_twist_01_r}}]
    constraintWeights [1,1]
   }
   n32 = Get thigh_rot_l()
   n33 = "Transform (Modify) Bone - Bone: dyn_belt_fr_base_l"(ComponentPose=@n31, Alpha=@n32, Translation="-5,3,0", Rotation="0,0,-15", Scale="1,1,1") {
    rotationMode=BMM_Additive
    rotationSpace=BCS_BoneSpace
    scaleMode=BMM_Additive
    scaleSpace=BCS_BoneSpace
    translationMode=BMM_Additive
    translationSpace=BCS_BoneSpace
    alphaBoolBlend {blendOption=Linear}
    alphaScaleBiasClamp {bMapRange=true, inRange {max=0, min=-100}, outRange {max=0, min=1}}
    boneToModify {boneName=dyn_belt_fr_base_l}
    rotation {roll=-15}
    translation=<-5 3>
   }
   n34 = "Transform (Modify) Bone - Bone: dyn_belt_side_base_l"(ComponentPose=@n33, Alpha=@n32, Translation="-3,0,0", Rotation="0,0,0", Scale="1,1,1") {
    rotationMode=BMM_Additive
    rotationSpace=BCS_BoneSpace
    scaleMode=BMM_Additive
    scaleSpace=BCS_BoneSpace
    translationMode=BMM_Additive
    translationSpace=BCS_BoneSpace
    alphaBoolBlend {blendOption=Linear}
    alphaScaleBiasClamp {bMapRange=true, inRange {max=0, min=-100}, outRange {max=0, min=1}}
    boneToModify {boneName=dyn_belt_side_base_l}
    translation {x=-3}
   }
   n35 = "Transform (Modify) Bone - Bone: dyn_belt_bk_base_l"(ComponentPose=@n34, Alpha=@n32, Translation="0,1.5,0", Rotation="0,0,0", Scale="1,1,1") {
    rotationMode=BMM_Additive
    rotationSpace=BCS_BoneSpace
    scaleMode=BMM_Additive
    scaleSpace=BCS_BoneSpace
    translationMode=BMM_Additive
    translationSpace=BCS_BoneSpace
    alphaBoolBlend {blendOption=Linear}
    alphaScaleBiasClamp {bMapRange=true, inRange {max=0, min=-100}, outRange {max=0, min=1}}
    boneToModify {boneName=dyn_belt_bk_base_l}
    translation {y=1.5}
   }
   n36 = Get thigh_rot_r()
   n37 = "Transform (Modify) Bone - Bone: dyn_belt_fr_base_r"(ComponentPose=@n35, Alpha=@n36, Translation="0,-12,0", Rotation="0,30,0", Scale="1,1,1") {
    rotationMode=BMM_Additive
    rotationSpace=BCS_BoneSpace
    scaleMode=BMM_Additive
    scaleSpace=BCS_BoneSpace
    translationMode=BMM_Additive
    translationSpace=BCS_BoneSpace
    alphaBoolBlend {blendOption=Linear}
    alphaScaleBiasClamp {bMapRange=true, inRange {max=0, min=-100}, outRange {max=0, min=1}}
    boneToModify {boneName=dyn_belt_fr_base_r}
    rotation {yaw=30}
    translation {y=-12}
   }
   n38 = "Transform (Modify) Bone - Bone: dyn_belt_side_base_r"(ComponentPose=@n37, Alpha=@n36, Translation="1,-1,-1", Rotation="0,-15,0", Scale="1,1,1") {
    rotationMode=BMM_Additive
    rotationSpace=BCS_BoneSpace
    scaleMode=BMM_Additive
    scaleSpace=BCS_BoneSpace
    translationMode=BMM_Additive
    translationSpace=BCS_BoneSpace
    alphaBoolBlend {blendOption=Linear}
    alphaScaleBiasClamp {bMapRange=true, inRange {max=0, min=-100}, outRange {max=0, min=1}}
    boneToModify {boneName=dyn_belt_side_base_r}
    rotation {yaw=-15}
    translation=<1 -1 -1>
   }
   n39 = "Transform (Modify) Bone - Bone: dyn_belt_bk_base_r"(ComponentPose=@n38, Alpha=@n36, Translation="1,-1.5,0", Rotation="-20,0,0", Scale="1,1,1") {
    rotationMode=BMM_Additive
    rotationSpace=BCS_BoneSpace
    scaleMode=BMM_Additive
    scaleSpace=BCS_BoneSpace
    translationMode=BMM_Additive
    translationSpace=BCS_BoneSpace
    alphaBoolBlend {blendOption=Linear}
    alphaScaleBiasClamp {bMapRange=true, inRange {max=0, min=-100}, outRange {max=0, min=1}}
    boneToModify {boneName=dyn_belt_bk_base_r}
    rotation {pitch=-20}
    translation=<1 -1.5>
   }
   n40 = "Constraint - dyn_canteen_base_r (PPPPPP)"!(
    ComponentPose=@n39
    Alpha=1
    ConstraintWeights_0=0.921
    ConstraintWeights_1:float=0.0022
    ConstraintWeights_2:float=0.0001
    ConstraintWeights_3:float=0.0067
    ConstraintWeights_4:float=0.0699
    ConstraintWeights_5:float=0.0001
   ) -> (Pose) {
    alphaBoolBlend {blendOption=Linear}
    boneToModify {boneName=dyn_canteen_base_r}
    constraintSetup [
     {targetBone {boneName=dyn_belt_side_r}}
     {targetBone {boneName=dyn_belt_fr_l}}
     {targetBone {boneName=dyn_belt_bk_l}}
     {targetBone {boneName=dyn_belt_bk_r}}
     {targetBone {boneName=dyn_belt_fr_r}}
     {targetBone {boneName=dyn_belt_side_l}}
    ]
    constraintWeights [1,1,1,1,1,1]
   }
   n41 = "Constraint - dyn_rope_a_base (PPPPPP)"!(
    ComponentPose=@n40
    Alpha=1
    ConstraintWeights_0=0.9308
    ConstraintWeights_1:float=0.0017
    ConstraintWeights_2:float=0.0001
    ConstraintWeights_3:float=0.0082
    ConstraintWeights_4:float=0.0591
    ConstraintWeights_5:float=0.0001
   ) -> (Pose) {
    alphaBoolBlend {blendOption=Linear}
    boneToModify {boneName=dyn_rope_a_base}
    constraintSetup [
     {targetBone {boneName=dyn_belt_side_r}}
     {targetBone {boneName=dyn_belt_fr_l}}
     {targetBone {boneName=dyn_belt_bk_l}}
     {targetBone {boneName=dyn_belt_bk_r}}
     {targetBone {boneName=dyn_belt_fr_r}}
     {targetBone {boneName=dyn_belt_side_l}}
    ]
    constraintWeights [1,1,1,1,1,1]
   }
   n42 = "Constraint - dyn_pouch_fr_base_r (PPPPPP)"!(
    ComponentPose=@n41
    Alpha=1
    ConstraintWeights_0=0.4548
    ConstraintWeights_1:float=0.128
    ConstraintWeights_2:float=0
    ConstraintWeights_3:float=0.0002
    ConstraintWeights_4:float=0.4144
    ConstraintWeights_5:float=0.0026
   ) -> (Pose) {
    alphaBoolBlend {blendOption=Linear}
    boneToModify {boneName=dyn_pouch_fr_base_r}
    constraintSetup [
     {targetBone {boneName=dyn_belt_side_r}}
     {targetBone {boneName=dyn_belt_fr_l}}
     {targetBone {boneName=dyn_belt_bk_l}}
     {targetBone {boneName=dyn_belt_bk_r}}
     {targetBone {boneName=dyn_belt_fr_r}}
     {targetBone {boneName=dyn_belt_side_l}}
    ]
    constraintWeights [1,1,1,1,1,1]
   }
   n43 = "Constraint - dyn_pouch_bk_base_r (PPPPPP)"!(
    ComponentPose=@n42
    Alpha=1
    ConstraintWeights_0=0.3407
    ConstraintWeights_1:float=0.0001
    ConstraintWeights_2:float=0.1027
    ConstraintWeights_3:float=0.5521
    ConstraintWeights_4:float=0.0035
    ConstraintWeights_5:float=0.0008
   ) -> (Pose) {
    alphaBoolBlend {blendOption=Linear}
    boneToModify {boneName=dyn_pouch_bk_base_r}
    constraintSetup [
     {targetBone {boneName=dyn_belt_side_r}}
     {targetBone {boneName=dyn_belt_fr_l}}
     {targetBone {boneName=dyn_belt_bk_l}}
     {targetBone {boneName=dyn_belt_bk_r}}
     {targetBone {boneName=dyn_belt_fr_r}}
     {targetBone {boneName=dyn_belt_side_l}}
    ]
    constraintWeights [1,1,1,1,1,1]
   }
   n44 = "Constraint - dyn_shirts_fr_base (PPPPP)"!(
    ComponentPose=@n43
    Alpha=1
    ConstraintWeights_0=0.0227
    ConstraintWeights_1:float=0.1714
    ConstraintWeights_2:float=0.1132
    ConstraintWeights_3:float=0.6763
    ConstraintWeights_4:float=0.0164
   ) -> (Pose) {
    alphaBoolBlend {blendOption=Linear}
    boneToModify {boneName=dyn_shirts_fr_base}
    constraintSetup [
     {targetBone {boneName=spine_01}}
     {targetBone {boneName=spine_02}}
     {targetBone {boneName=spine_03}}
     {targetBone {boneName=spine_04}}
     {targetBone {boneName=pelvis}}
    ]
    constraintWeights [1,1,1,1,1]
   }
   n45 = "Constraint - dyn_shirts_fr_base_l (PPPPP)"!(
    ComponentPose=@n44
    Alpha=1
    ConstraintWeights_0=0.024
    ConstraintWeights_1:float=0.1419
    ConstraintWeights_2:float=0.1024
    ConstraintWeights_3:float=0.7184
    ConstraintWeights_4:float=0.0134
   ) -> (Pose) {
    alphaBoolBlend {blendOption=Linear}
    boneToModify {boneName=dyn_shirts_fr_base_l}
    constraintSetup [
     {targetBone {boneName=spine_01}}
     {targetBone {boneName=spine_02}}
     {targetBone {boneName=spine_03}}
     {targetBone {boneName=spine_04}}
     {targetBone {boneName=pelvis}}
    ]
    constraintWeights [1,1,1,1,1]
   }
   n46 = "Constraint - dyn_shirts_fr_base_r (PPPPP)"!(
    ComponentPose=@n45
    Alpha=1
    ConstraintWeights_0=0.0269
    ConstraintWeights_1:float=0.1864
    ConstraintWeights_2:float=0.1187
    ConstraintWeights_3:float=0.6468
    ConstraintWeights_4:float=0.0213
   ) -> (Pose) {
    alphaBoolBlend {blendOption=Linear}
    boneToModify {boneName=dyn_shirts_fr_base_r}
    constraintSetup [
     {targetBone {boneName=spine_01}}
     {targetBone {boneName=spine_02}}
     {targetBone {boneName=spine_03}}
     {targetBone {boneName=spine_04}}
     {targetBone {boneName=pelvis}}
    ]
    constraintWeights [1,1,1,1,1]
   }
   n47 = "Constraint - dyn_shirts_bk_base_l (PPPPP)"!(
    ComponentPose=@n46
    Alpha=1
    ConstraintWeights_0=0.0707
    ConstraintWeights_1:float=0.3129
    ConstraintWeights_2:float=0.1659
    ConstraintWeights_3:float=0.4009
    ConstraintWeights_4:float=0.0497
   ) -> (Pose) {
    alphaBoolBlend {blendOption=Linear}
    boneToModify {boneName=dyn_shirts_bk_base_l}
    constraintSetup [
     {targetBone {boneName=spine_01}}
     {targetBone {boneName=spine_02}}
     {targetBone {boneName=spine_03}}
     {targetBone {boneName=spine_04}}
     {targetBone {boneName=clavicle_l}}
    ]
    constraintWeights [1,1,1,1,1]
   }
   n48 = "Constraint - dyn_shirts_bk_base (PPPPP)"!(
    ComponentPose=@n47
    Alpha=1
    ConstraintWeights_0=0.0794
    ConstraintWeights_1:float=0.2121
    ConstraintWeights_2:float=0.1554
    ConstraintWeights_3:float=0.3853
    ConstraintWeights_4:float=0.1678
   ) -> (Pose) {
    alphaBoolBlend {blendOption=Linear}
    boneToModify {boneName=dyn_shirts_bk_base}
    constraintSetup [
     {targetBone {boneName=clavicle_r}}
     {targetBone {boneName=spine_02}}
     {targetBone {boneName=spine_03}}
     {targetBone {boneName=spine_04}}
     {targetBone {boneName=spine_05}}
    ]
    constraintWeights [1,1,1,1,1]
   }
   n49 = "Transform (Modify) Bone - Bone: dyn_sword_base"(ComponentPose=@n48, Alpha=Get thigh_rot_l(), Translation="0,0,0", Rotation="0,-90,0", Scale="1,1,1") {
    rotationMode=BMM_Additive
    rotationSpace=BCS_BoneSpace
    scaleMode=BMM_Additive
    scaleSpace=BCS_BoneSpace
    translationMode=BMM_Additive
    translationSpace=BCS_BoneSpace
    alphaBoolBlend {blendOption=Linear}
    alphaScaleBiasClamp {bMapRange=true, inRange {max=0, min=-100}, outRange {max=0, min=1}}
    boneToModify {boneName=dyn_sword_base}
    rotation {yaw=-90}
   }
   n50 = RigidBody(ComponentPose=@n49) {
    bClampLinearTranslationLimitToRefPose=true
    bOverrideWorldGravity=true
    overridePhysicsAsset=&/F_MED_Dino_Hunter_Shirts_Physics
    simulationSpace=BaseBoneSpace
    alphaBoolBlend {blendOption=Linear}
    baseBoneRef {boneName=pelvis}
    componentLinearAccScale=<0.35 0.35 0.35>
    componentLinearVelScale=<0.35 0.35 0.35>
    overrideWorldGravity {z=-980}
   }
   n51 = RigidBody(ComponentPose=@n50) {
    bClampLinearTranslationLimitToRefPose=true
    bOverrideWorldGravity=true
    overridePhysicsAsset=&/F_MED_Dino_Hunter_Sword_Physics
    simulationSpace=BaseBoneSpace
    alphaBoolBlend {blendOption=Linear}
    baseBoneRef {boneName=thigh_l}
    componentLinearAccScale=<0.35 0.35 0.35>
    componentLinearVelScale=<0.35 0.35 0.35>
    overrideWorldGravity {z=-980}
   }
   n52 = RigidBody(ComponentPose=@n51) {
    bClampLinearTranslationLimitToRefPose=true
    bOverrideWorldGravity=true
    overridePhysicsAsset=&/F_MED_Dino_Hunter_Necklace_Physics
    simulationSpace=BaseBoneSpace
    alphaBoolBlend {blendOption=Linear}
    baseBoneRef {boneName=spine_03}
    componentLinearAccScale=<0.4 0.4 0.4>
    componentLinearVelScale=<0.4 0.4 0.4>
    overrideWorldGravity {z=-980}
   }
   n53 = RigidBody(ComponentPose=@n52) {
    bClampLinearTranslationLimitToRefPose=true
    bOverrideWorldGravity=true
    overridePhysicsAsset=&/F_MED_Dino_Hunter_Belt_Physics
    simulationSpace=BaseBoneSpace
    alphaBoolBlend {blendOption=Linear}
    baseBoneRef {boneName=pelvis}
    componentLinearAccScale=<0.35 0.35 0.35>
    componentLinearVelScale=<0.35 0.35 0.35>
    overrideWorldGravity {z=-980}
   }
   n54 = Anim Dynamics - Bone: dyn_swordTeeth_a(ComponentPose=@n53, Alpha=1, ExternalForce="0,0,0") {
    angularDampingOverride=1
    angularSpringConstant=1500
    bAngularSpring=true
    bGravityOverrideInSimSpace=true
    bOverrideAngularDamping=true
    gravityScale=0.35
    simulationSpace=RootRelative
    alphaBoolBlend {blendOption=Linear}
    boundBone {boneName=dyn_swordTeeth_a}
    physicsBodyDefinitions [{
      sphereCollisionRadius=0
      boundBone {boneName=dyn_swordTeeth_a}
      boxExtents=<20 10 20>
      constraintSetup {twistAxis=AxisY, angularLimitsMax=<25 0 15>, angularLimitsMin=<-25 0 -15>, angularTarget=<0 0 0>}
      localJointOffset=<0 10 0>
     }]
   }
   n55 = Anim Dynamics - Bone: dyn_swordTeeth_b(ComponentPose=@n54, Alpha=0.75, ExternalForce="0,0,0") {
    angularDampingOverride=1
    angularSpringConstant=1500
    bAngularSpring=true
    bGravityOverrideInSimSpace=true
    bOverrideAngularDamping=true
    gravityScale=0.35
    simulationSpace=RootRelative
    alphaBoolBlend {blendOption=Linear}
    boundBone {boneName=dyn_swordTeeth_b}
    physicsBodyDefinitions [{
      sphereCollisionRadius=0
      boundBone {boneName=dyn_swordTeeth_b}
      boxExtents=<20 10 20>
      constraintSetup {twistAxis=AxisY, angularLimitsMax=<20 0 15>, angularLimitsMin=<-20 0 -15>, angularTarget=<0 0 0>}
      localJointOffset=<0 10 0>
     }]
   }
   n56 = Anim Dynamics - Bone: dyn_swordTeeth_c(ComponentPose=@n55, Alpha=0.5, ExternalForce="0,0,0") {
    angularDampingOverride=1
    angularSpringConstant=1500
    bAngularSpring=true
    bGravityOverrideInSimSpace=true
    bOverrideAngularDamping=true
    gravityScale=0.35
    simulationSpace=RootRelative
    alphaBoolBlend {blendOption=Linear}
    boundBone {boneName=dyn_swordTeeth_c}
    physicsBodyDefinitions [{
      sphereCollisionRadius=0
      boundBone {boneName=dyn_swordTeeth_c}
      boxExtents=<20 10 20>
      constraintSetup {twistAxis=AxisY, angularLimitsMax=<15 0 15>, angularLimitsMin=<-15 0 -15>, angularTarget=<0 0 0>}
      localJointOffset=<0 10 0>
     }]
   }
   n57 = Anim Dynamics - Bone: dyn_shoeString_a_l(ComponentPose=@n56, Alpha=1, ExternalForce="0,0,0") {
    angularDampingOverride=1
    angularSpringConstant=1500
    bAngularSpring=true
    bGravityOverrideInSimSpace=true
    bOverrideAngularDamping=true
    gravityScale=0.35
    simulationSpace=RootRelative
    alphaBoolBlend {blendOption=Linear}
    boundBone {boneName=dyn_shoeString_a_l}
    physicsBodyDefinitions [{
      sphereCollisionRadius=0
      boundBone {boneName=dyn_shoeString_a_l}
      boxExtents=<10 20 20>
      constraintSetup {angularLimitsMax=<10 0 0>, angularLimitsMin=<-10 -25 0>, angularTarget=<0 0 0>}
      localJointOffset=<10 0 0>
     }]
   }
   n58 = Anim Dynamics - Bone: dyn_shoeString_b_l(ComponentPose=@n57, Alpha=1, ExternalForce="0,0,0") {
    angularDampingOverride=1
    angularSpringConstant=1500
    bAngularSpring=true
    bGravityOverrideInSimSpace=true
    bOverrideAngularDamping=true
    gravityScale=0.35
    simulationSpace=RootRelative
    alphaBoolBlend {blendOption=Linear}
    boundBone {boneName=dyn_shoeString_b_l}
    physicsBodyDefinitions [{
      sphereCollisionRadius=0
      boundBone {boneName=dyn_shoeString_b_l}
      boxExtents=<10 20 20>
      constraintSetup {angularLimitsMax=<10 25 0>, angularLimitsMin=<-10 0 0>, angularTarget=<0 0 0>}
      localJointOffset=<10 0 0>
     }]
   }
   n59 = Anim Dynamics - Bone: dyn_shoeString_b_r(ComponentPose=@n58, Alpha=1, ExternalForce="0,0,0") {
    angularDampingOverride=1
    angularSpringConstant=1500
    bAngularSpring=true
    bGravityOverrideInSimSpace=true
    bOverrideAngularDamping=true
    gravityScale=0.35
    simulationSpace=RootRelative
    alphaBoolBlend {blendOption=Linear}
    boundBone {boneName=dyn_shoeString_b_r}
    physicsBodyDefinitions [{
      sphereCollisionRadius=0
      boundBone {boneName=dyn_shoeString_b_r}
      boxExtents=<10 20 20>
      constraintSetup {angularLimitsMax=<10 0 0>, angularLimitsMin=<-10 -25 0>, angularTarget=<0 0 0>}
      localJointOffset=<-10 0 0>
     }]
   }
   n60 = Anim Dynamics - Bone: dyn_shoeString_a_r(ComponentPose=@n59, Alpha=1, ExternalForce="0,0,0") {
    angularDampingOverride=1
    angularSpringConstant=1500
    bAngularSpring=true
    bGravityOverrideInSimSpace=true
    bOverrideAngularDamping=true
    gravityScale=0.35
    simulationSpace=RootRelative
    alphaBoolBlend {blendOption=Linear}
    boundBone {boneName=dyn_shoeString_a_r}
    physicsBodyDefinitions [{
      sphereCollisionRadius=0
      boundBone {boneName=dyn_shoeString_a_r}
      boxExtents=<10 20 20>
      constraintSetup {angularLimitsMax=<10 0 0>, angularLimitsMin=<-10 -25 0>, angularTarget=<0 0 0>}
      localJointOffset=<-10 0 0>
     }]
   }
   Output Pose(Result=Component To Local(ComponentPose=@n60)) {layerGroup=None}
  ]}
 EventGraph {nodes [
   Event BlueprintUpdateAnimation()
   > Sequence!(execute) -> (then_0, then_1, then_2:Exec, then_3:Exec, then_4:Exec, then_5:Exec, then_6:Exec)
   > n1 = Sequence()
   n2 = GetDeltaTransformFromRefPose(self=GetOwnerSkeletalMeshComponent(), BoneName=lowerarm_l, BaseName=upperarm_l)
   > Set elbow_rot_l(elbow_rot_l:double=BreakRotator(InRot=BreakTransform(InTransform=@n2).Rotation).Yaw) -> (Output_Get:double)
   n3 = GetDeltaTransformFromRefPose(self=GetOwnerSkeletalMeshComponent(), BoneName=thigh_r, BaseName=pelvis)
   Set thigh_rot_r(execute=@n1.then_3, thigh_rot_r=BreakRotator(InRot=BreakTransform(InTransform=@n3).Rotation).Yaw)
   n4 = GetDeltaTransformFromRefPose(self=GetOwnerSkeletalMeshComponent(), BoneName=calf_r, BaseName=thigh_r)
   Set knee_rot_r(execute=@n1.then_5, knee_rot_r:double=BreakRotator(InRot=BreakTransform(InTransform=@n4).Rotation).Yaw) -> (Output_Get:double)
   n5 = GetDeltaTransformFromRefPose(self=GetOwnerSkeletalMeshComponent(), BoneName=calf_l, BaseName=thigh_l)
   Set knee_rot_l(execute=@n1.then_4, knee_rot_l:double=BreakRotator(InRot=BreakTransform(InTransform=@n5).Rotation).Yaw) -> (Output_Get:double)
   n6 = GetDeltaTransformFromRefPose(self=GetOwnerSkeletalMeshComponent(), BoneName=thigh_l, BaseName=pelvis)
   Set thigh_rot_l(execute=@n1.then_2, thigh_rot_l=BreakRotator(InRot=BreakTransform(InTransform=@n6).Rotation).Yaw)
   n7 = GetDeltaTransformFromRefPose(self=GetOwnerSkeletalMeshComponent(), BoneName=upperarm_r, BaseName=clavicle_r)
   Set uparm_rot_r(execute=@n1.then_6, uparm_rot_r:double=BreakRotator(InRot=BreakTransform(InTransform=@n7).Rotation).Yaw) -> (Output_Get:double)
   n8 = GetDeltaTransformFromRefPose(self=GetOwnerSkeletalMeshComponent(), BoneName=lowerarm_r, BaseName=upperarm_r)
   Set elbow_rot_r(execute=@n1.then_1, elbow_rot_r:double=BreakRotator(InRot=BreakTransform(InTransform=@n8).Rotation).Yaw) -> (Output_Get:double)
  ]}
]
variables [
 elbow_rot_l {type=number}
 elbow_rot_r {type=number}
 knee_rot_l {type=number}
 knee_rot_r {type=number}
 thigh_rot_l {type=number}
 thigh_rot_r {type=number}
 uparm_rot_r {type=number}
]
