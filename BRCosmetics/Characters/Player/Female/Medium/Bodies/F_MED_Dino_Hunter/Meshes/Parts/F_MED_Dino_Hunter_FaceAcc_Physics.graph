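# Physics asset for the face accessory part of the female medium Dino Hunter player cosmetic, covering a three-bone dynamic hair chain.
# The head and each hair bone carry a single capsule body.
# Constraints link the hair bones in series from the head, with identical tight swing and twist limits and the widest range on the first swing axis.
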PhysicsAsset /BRCosmetics/Characters/Player/Female/Medium/Bodies/F_MED_Dino_Hunter/Meshes/Parts/F_MED_Dino_Hunter_FaceAcc_Physics
bodies {
 dyn_hair_main_a [{shapeType=Capsule, length=0.5, radius=2, center=<2 0 0>, rotation=<rot 90 0 0>}]
 dyn_hair_main_b [{shapeType=Capsule, length=0.5, radius=2, center=<2 0 0>, rotation=<rot 90 0 0>}]
 dyn_hair_main_c [{shapeType=Capsule, length=0.5, radius=2, center=<2 0 0>, rotation=<rot 90 0 0>}]
 head [{shapeType=Capsule, length=1, radius=1}]
}
constraints [
 {bone1Name=dyn_hair_main_a, bone2Name=head, swing1LimitDegrees=10, swing2LimitDegrees=3, twistLimitDegrees=3}
 {bone1Name=dyn_hair_main_b, bone2Name=dyn_hair_main_a, swing1LimitDegrees=10, swing2LimitDegrees=3, twistLimitDegrees=3}
 {bone1Name=dyn_hair_main_c, bone2Name=dyn_hair_main_b, swing1LimitDegrees=10, swing2LimitDegrees=3, twistLimitDegrees=3}
]
